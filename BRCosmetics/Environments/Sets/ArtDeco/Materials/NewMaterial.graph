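# NewMaterial is a simple Default Lit material in the ArtDeco environment set of BRCosmetics, still under its default asset name.
# Its graph holds only a dark, near-neutral grey color constant and two identical mid-low scalar constants.
# It has no textures, parameters or math, and the file shows no links from the constants to material inputs.
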Material /BRCosmetics/Environments/Sets/ArtDeco/Materials/NewMaterial
graph {
 properties {pixelDepthOffsetMode=PDOM_Legacy, usedShadingModels=Default Lit}
 nodes [
  Constant3Vector() {constant=<rgba 0.0521 0.0491 0.0491 0>}
  Constant() {r=0.4}
  Constant() {r=0.4}
 ]
}
